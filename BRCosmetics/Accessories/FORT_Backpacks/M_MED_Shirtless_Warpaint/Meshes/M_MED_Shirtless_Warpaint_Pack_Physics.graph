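# Physics asset for the Shirtless Warpaint backpack cosmetic, giving its dangling parts secondary motion.
# Box bodies cover the pack, a knife, a pouch and two mirrored rows of small bullet bones on the left and right.
# Every dangling body is constrained directly to the pack rather than chained, with narrow swing and twist limits on the pouch and bullets and looser swing on the knife.
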PhysicsAsset /BRCosmetics/Accessories/FORT_Backpacks/M_MED_Shirtless_Warpaint/Meshes/M_MED_Shirtless_Warpaint_Pack_Physics
bodies {
 dyn_bullets_l_a [{shapeType=Box, extentX=5, extentY=1, extentZ=1, center=<-1 0 0>}]
 dyn_bullets_l_b [{shapeType=Box, extentX=5, extentY=1, extentZ=1, center=<-1 0 0>}]
 dyn_bullets_l_c [{shapeType=Box, extentX=5, extentY=1, extentZ=1, center=<-1 0 0>}]
 dyn_bullets_l_d [{shapeType=Box, extentX=5, extentY=1, extentZ=1, center=<-1 0 0>}]
 dyn_bullets_l_e [{shapeType=Box, extentX=5, extentY=1, extentZ=1, center=<-1 0 0>}]
 dyn_bullets_l_f [{shapeType=Box, extentX=5, extentY=1, extentZ=1, center=<-1 0 0>}]
 dyn_bullets_l_g [{shapeType=Box, extentX=5, extentY=1, extentZ=1, center=<-1 0 0>}]
 dyn_bullets_l_h [{shapeType=Box, extentX=5, extentY=1, extentZ=1, center=<-1 0 0>}]
 dyn_bullets_l_i [{shapeType=Box, extentX=5, extentY=1, extentZ=1, center=<-1 0 0>}]
 dyn_bullets_r_a [{shapeType=Box, extentX=5, extentY=1, extentZ=1, center=<1 0 0>}]
 dyn_bullets_r_b [{shapeType=Box, extentX=5, extentY=1, extentZ=1, center=<1 0 0>}]
 dyn_bullets_r_c [{shapeType=Box, extentX=5, extentY=1, extentZ=1, center=<1 0 0>}]
 dyn_bullets_r_d [{shapeType=Box, extentX=5, extentY=1, extentZ=1, center=<1 0 0>}]
 dyn_bullets_r_e [{shapeType=Box, extentX=5, extentY=1, extentZ=1, center=<1 0 0>}]
 dyn_bullets_r_f [{shapeType=Box, extentX=5, extentY=1, extentZ=1, center=<1 0 0>}]
 dyn_bullets_r_g [{shapeType=Box, extentX=5, extentY=1, extentZ=1, center=<1 0 0>}]
 dyn_bullets_r_h [{shapeType=Box, extentX=5, extentY=1, extentZ=1, center=<1 0 0>}]
 dyn_bullets_r_i [{shapeType=Box, extentX=5, extentY=1, extentZ=1, center=<1 0 0>}]
 dyn_knife [{shapeType=Box, extentX=38.3963, extentY=5.4735, extentZ=10.6985, center=<-0.1901 -0.4168 0.5603>}]
 dyn_pouch [{shapeType=Box, extentX=14.8821, extentY=5.8009, extentZ=12.6619, center=<4.3776 1.941 1.0514>}]
 pack [{shapeType=Box, extentX=19.431, extentY=16.3883, extentZ=30.196, center=<-0.2935 -7.4218 -4.3238>}]
}
constraints [
 {bone1Name=dyn_pouch, bone2Name=pack, swing1LimitDegrees=5, swing2LimitDegrees=2, twistLimitDegrees=10}
 {bone1Name=dyn_knife, bone2Name=pack, swing1LimitDegrees=25, swing2LimitDegrees=25, twistLimitDegrees=12}
 {bone1Name=dyn_bullets_r_i, bone2Name=pack, swing1LimitDegrees=9, swing2LimitDegrees=9, twistLimitDegrees=9}
 {bone1Name=dyn_bullets_l_b, bone2Name=pack, swing1LimitDegrees=9, swing2LimitDegrees=9, twistLimitDegrees=9}
 {bone1Name=dyn_bullets_l_c, bone2Name=pack, swing1LimitDegrees=9, swing2LimitDegrees=9, twistLimitDegrees=9}
 {bone1Name=dyn_bullets_l_d, bone2Name=pack, swing1LimitDegrees=9, swing2LimitDegrees=9, twistLimitDegrees=9}
 {bone1Name=dyn_bullets_l_e, bone2Name=pack, swing1LimitDegrees=9, swing2LimitDegrees=9, twistLimitDegrees=9}
 {bone1Name=dyn_bullets_l_f, bone2Name=pack, swing1LimitDegrees=9, swing2LimitDegrees=9, twistLimitDegrees=9}
 {bone1Name=dyn_bullets_l_g, bone2Name=pack, swing1LimitDegrees=9, swing2LimitDegrees=9, twistLimitDegrees=9}
 {bone1Name=dyn_bullets_l_h, bone2Name=pack, swing1LimitDegrees=9, swing2LimitDegrees=9, twistLimitDegrees=9}
 {bone1Name=dyn_bullets_l_i, bone2Name=pack, swing1LimitDegrees=9, swing2LimitDegrees=9, twistLimitDegrees=9}
 {bone1Name=dyn_bullets_r_a, bone2Name=pack, swing1LimitDegrees=9, swing2LimitDegrees=9, twistLimitDegrees=9}
 {bone1Name=dyn_bullets_r_b, bone2Name=pack, swing1LimitDegrees=9, swing2LimitDegrees=9, twistLimitDegrees=9}
 {bone1Name=dyn_bullets_r_c, bone2Name=pack, swing1LimitDegrees=9, swing2LimitDegrees=9, twistLimitDegrees=9}
 {bone1Name=dyn_bullets_r_d, bone2Name=pack, swing1LimitDegrees=9, swing2LimitDegrees=9, twistLimitDegrees=9}
 {bone1Name=dyn_bullets_r_e, bone2Name=pack, swing1LimitDegrees=9, swing2LimitDegrees=9, twistLimitDegrees=9}
 {bone1Name=dyn_bullets_r_f, bone2Name=pack, swing1LimitDegrees=9, swing2LimitDegrees=9, twistLimitDegrees=9}
 {bone1Name=dyn_bullets_r_g, bone2Name=pack, swing1LimitDegrees=9, swing2LimitDegrees=9, twistLimitDegrees=9}
 {bone1Name=dyn_bullets_r_h, bone2Name=pack, swing1LimitDegrees=9, swing2LimitDegrees=9, twistLimitDegrees=9}
 {bone1Name=dyn_bullets_l_a, bone2Name=pack, swing1LimitDegrees=9, swing2LimitDegrees=9, twistLimitDegrees=9}
]
